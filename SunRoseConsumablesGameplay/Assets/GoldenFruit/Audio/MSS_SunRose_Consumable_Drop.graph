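# One-shot mono MetaSound Source for the SunRose (Golden Fruit) consumable drop sound.
# On play, a random array pick feeds a mono Wave Player, which fires OnFinished when the wave ends.
# The Sounds array defaults to a single wave, so there is no variation by default.
# It uses the Fort_Consumables sound class and borrows the consumable-use Athena attenuation.
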MetaSoundSource /SunRoseConsumablesGameplay/Assets/GoldenFruit/Audio/MSS_SunRose_Consumable_Drop
inputs [
 Sounds {type=WaveAsset:Array}
 UE.Source.OnPlay {type=Trigger}
]
interfaces [UE.OutputFormat.Mono:1.0, UE.Source.OneShot:1.0, UE.Source:1.1]
outputs [
 "UE.OutputFormat.Mono.Audio:0" {type=Audio}
 UE.Source.OneShot.OnFinished {type=Trigger}
]
pages [Default {
  input_defaults {UE.Source.OnPlay=null, Sounds [/SunRoseConsumablesGameplay/Assets/GoldenFruit/Audio/wav/SunRose_Consumable_Drop_01]}
  nodes [
   ArrayRandomGet_0 = Array.Random Get:WaveAsset:Array(In Array=^Sounds, Next=^UE.Source.OnPlay)
   WavePlayer_0 = UE.Wave Player:Mono(Play=@ArrayRandomGet_0.On Next, Wave Asset=@ArrayRandomGet_0.Value)
  ]
  links [{dst="@UE.Source.OneShot.OnFinished", src=WavePlayer_0.On Finished}, {dst="@UE.OutputFormat.Mono.Audio:0", src=WavePlayer_0.Out Mono}]
 }]
properties {
 assetImportData=&self:AssetImportData
 attenuationSettings=&Fort_Consumable_Use_Athena_Att
 maxDistance=2250
 sampleRate=48000
 soundClassObject=&Fort_Consumables
}
