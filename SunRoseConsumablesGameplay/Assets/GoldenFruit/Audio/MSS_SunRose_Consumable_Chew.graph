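# One-shot mono MetaSound source for the chewing sound of the SunRose GoldenFruit consumable.
# It passes an array of four chew wave assets and the play trigger into a nested MSS_SunRose_Consumable_Drop graph, reusing that graph's playback logic.
# The nested graph's audio and finished trigger feed straight to the outputs.
# Playback uses the Fortnite consumables sound class and consumable-use attenuation, with boosted volume.
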MetaSoundSource /SunRoseConsumablesGameplay/Assets/GoldenFruit/Audio/MSS_SunRose_Consumable_Chew
inputs [
 Chew {type=WaveAsset:Array}
 UE.Source.OnPlay {type=Trigger}
]
interfaces [UE.OutputFormat.Mono:1.0, UE.Source.OneShot:1.0, UE.Source:1.1]
outputs [
 "UE.OutputFormat.Mono.Audio:0" {type=Audio}
 UE.Source.OneShot.OnFinished {type=Trigger}
]
pages [Default {
  input_defaults {
   UE.Source.OnPlay=null
   Chew [
    /SunRoseConsumablesGameplay/Assets/GoldenFruit/Audio/wav/SunRose_Consumable_Chew_01
    /SunRoseConsumablesGameplay/Assets/GoldenFruit/Audio/wav/SunRose_Consumable_Chew_02
    /SunRoseConsumablesGameplay/Assets/GoldenFruit/Audio/wav/SunRose_Consumable_Chew_03
    /SunRoseConsumablesGameplay/Assets/GoldenFruit/Audio/wav/SunRose_Consumable_Chew_04
   ]
  }
  nodes [MSS_SunRose_Consumable_Drop_0 = &/MSS_SunRose_Consumable_Drop(Sounds=^Chew, UE.Source.OnPlay=^UE.Source.OnPlay)]
  links [
   {dst="@UE.OutputFormat.Mono.Audio:0", src=MSS_SunRose_Consumable_Drop_0.UE.OutputFormat.Mono.Audio:0}
   {dst="@UE.Source.OneShot.OnFinished", src=MSS_SunRose_Consumable_Drop_0.UE.Source.OneShot.OnFinished}
  ]
 }]
properties {
 assetImportData=&self:AssetImportData
 attenuationSettings=&Fort_Consumable_Use_Athena_Att
 maxDistance=2250
 sampleRate=48000
 soundClassObject=&Fort_Consumables
 volume=3
}
