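# Stereo one-shot MetaSound source that plays the equip sound for the SunRose GoldenFruit consumable.
# The base sound reuses the mono SunRose Drop MetaSound with equip wave input, feeding both channels.
# A short trigger delay starts a random stereo sparkle layer, mixed in at low gain.
# A trigger accumulator signals finished once both layers complete, and a concurrency override limits overlapping instances.
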MetaSoundSource /SunRoseConsumablesGameplay/Assets/GoldenFruit/Audio/MSS_SunRose_Consumable_Equip
inputs [
 Sounds {type=WaveAsset:Array}
 Sparkle Layer {type=WaveAsset:Array}
 UE.Source.OnPlay {type=Trigger}
]
interfaces [UE.OutputFormat.Stereo:1.0, UE.Source.OneShot:1.0, UE.Source:1.1]
outputs [
 "UE.OutputFormat.Stereo.Audio:0" {type=Audio}
 "UE.OutputFormat.Stereo.Audio:1" {type=Audio}
 UE.Source.OneShot.OnFinished {type=Trigger}
]
pages [Default {
  input_defaults {
   UE.Source.OnPlay=null
   Sounds [/SunRoseConsumablesGameplay/Assets/GoldenFruit/Audio/wav/SunRose_Consumable_Equip_01]
   Sparkle Layer [
    /SunRoseConsumablesGameplay/Assets/GoldenFruit/Audio/wav/SunRose_Consumable_Equip_Sparkle_01
    /SunRoseConsumablesGameplay/Assets/GoldenFruit/Audio/wav/SunRose_Consumable_Equip_Sparkle_02
    /SunRoseConsumablesGameplay/Assets/GoldenFruit/Audio/wav/SunRose_Consumable_Equip_Sparkle_03
   ]
  }
  nodes [
   TriggerDelay_0 = UE.Trigger Delay(In=^UE.Source.OnPlay) {also {defaults {Delay Time=0.3}}}
   n1 = &MSS_PlayRandomSound_Stereo(Sounds=^Sparkle Layer, UE.Source.OnPlay=@TriggerDelay_0.Out)
   n2 = &/MSS_SunRose_Consumable_Drop(Sounds=^Sounds, UE.Source.OnPlay=^UE.Source.OnPlay)
   AudioMixerAudioMixerStereo2_0 = "AudioMixer.Audio Mixer (Stereo, 2)"(
    In 1 L=@n1."UE.OutputFormat.Stereo.Audio:0"
    In 1 R=@n1."UE.OutputFormat.Stereo.Audio:1"
    In 0 L=@n2."UE.OutputFormat.Mono.Audio:0"
    In 0 R=@n2."UE.OutputFormat.Mono.Audio:0"
   ) {also {defaults {Gain 1=0.15}}}
   TriggerAccumulatorTriggerAccumulate2_0 = "TriggerAccumulator.Trigger Accumulate (2)"(In 1=@n1.UE.Source.OneShot.OnFinished, In 0=@n2.UE.Source.OneShot.OnFinished)
  ]
  links [
   {dst="@UE.OutputFormat.Stereo.Audio:0", src=AudioMixerAudioMixerStereo2_0.Out L}
   {dst="@UE.OutputFormat.Stereo.Audio:1", src=AudioMixerAudioMixerStereo2_0.Out R}
   {dst="@UE.Source.OneShot.OnFinished", src=TriggerAccumulatorTriggerAccumulate2_0.Out}
  ]
 }]
properties {
 assetImportData=&self:AssetImportData
 attenuationSettings=&Fort_Consumable_Use_Athena_Att
 bOverrideConcurrency=true
 maxDistance=2250
 numChannels=2
 outputFormat=Stereo
 sampleRate=48000
 soundClassObject=&Fort_Consumables
 concurrencyOverrides {maxCount=2, retriggerTime=0.2}
}
