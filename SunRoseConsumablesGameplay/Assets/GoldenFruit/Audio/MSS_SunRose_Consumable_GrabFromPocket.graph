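# MetaSound Source preset for the grab-from-pocket sound of the SunRose Golden Fruit consumable.
# It inherits its graph from MSS_SunRose_Consumable_Drop and sets the Sounds array to a single wave, SunRose_Consumable_Cook_01.
# Output is a mono one-shot in the Fort_Consumables sound class, attenuated by the Athena consumable-use settings, with a raised volume.
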MetaSoundSource /SunRoseConsumablesGameplay/Assets/GoldenFruit/Audio/MSS_SunRose_Consumable_GrabFromPocket
preset_parent=&/MSS_SunRose_Consumable_Drop
inputs [
 Sounds {type=WaveAsset:Array}
 UE.Source.OnPlay {type=Trigger}
]
interfaces [UE.OutputFormat.Mono:1.0, UE.Source.OneShot:1.0, UE.Source:1.1]
outputs [
 "UE.OutputFormat.Mono.Audio:0" {type=Audio}
 UE.Source.OneShot.OnFinished {type=Trigger}
]
pages [Default {input_defaults {UE.Source.OnPlay=null, Sounds [/SunRoseConsumablesGameplay/Assets/GoldenFruit/Audio/wav/SunRose_Consumable_Cook_01]}}]
properties {
 assetImportData=&self:AssetImportData
 attenuationSettings=&Fort_Consumable_Use_Athena_Att
 maxDistance=2250
 sampleRate=48000
 soundClassObject=&Fort_Consumables
 volume=3.5
}
