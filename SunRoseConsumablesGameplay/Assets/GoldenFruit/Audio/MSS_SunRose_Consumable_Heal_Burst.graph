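# Stereo one-shot MetaSound source for the Golden Fruit (SunRose) consumable's heal burst.
# On play, a delayed repeating trigger periodically draws from the Sounds array, currently one wave, and replays it through a wave player.
# A parallel MSP_Fade patch drives the mixer gain and fires OnFinished when done, bounding the repeating loop.
# It uses consumable attenuation and sound class, with owner-limited concurrency that stops the oldest.
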MetaSoundSource /SunRoseConsumablesGameplay/Assets/GoldenFruit/Audio/MSS_SunRose_Consumable_Heal_Burst
inputs [
 Sounds {type=WaveAsset:Array}
 UE.Source.OnPlay {type=Trigger}
]
interfaces [UE.OutputFormat.Stereo:1.0, UE.Source.OneShot:1.0, UE.Source:1.1]
outputs [
 "UE.OutputFormat.Stereo.Audio:0" {type=Audio}
 "UE.OutputFormat.Stereo.Audio:1" {type=Audio}
 UE.Source.OneShot.OnFinished {type=Trigger}
]
pages [Default {
  input_defaults {UE.Source.OnPlay=null, Sounds [/SunRoseConsumablesGameplay/Assets/GoldenFruit/Audio/wav/SunRose_Consumable_Heal_Burst_OS_01]}
  nodes [
   MSP_Fade_0 = &MSP_Fade(In=^UE.Source.OnPlay) {also {defaults {Fade Time=15, bInvertOutput="<MetaSoundBuilderResult.SUCCEEDED: 0>"}}}
   TriggerDelay_0 = UE.Trigger Delay(In=^UE.Source.OnPlay)
   ArrayRandomGet_0 = Array.Random Get:WaveAsset:Array(In Array=^Sounds, Next=UE.TriggerRepeat(Start=@TriggerDelay_0.Out) {also {defaults {Period=2}}}.RepeatOut)
   WavePlayer_0 = UE.Wave Player:Stereo(Play=@ArrayRandomGet_0.On Next, Wave Asset=@ArrayRandomGet_0.Value)
   AudioMixerAudioMixerStereo2_0 = "AudioMixer.Audio Mixer (Stereo, 2)"(Gain 0=@MSP_Fade_0.Value Out, In 0 L=@WavePlayer_0.Out Left, In 0 R=@WavePlayer_0.Out Right) {also {defaults {Gain 1=0.05}}}
  ]
  links [
   {dst="@UE.OutputFormat.Stereo.Audio:0", src=AudioMixerAudioMixerStereo2_0.Out L}
   {dst="@UE.OutputFormat.Stereo.Audio:1", src=AudioMixerAudioMixerStereo2_0.Out R}
   {dst="@UE.Source.OneShot.OnFinished", src=MSP_Fade_0.On Done}
  ]
 }]
properties {
 assetImportData=&self:AssetImportData
 attenuationSettings=&Fort_Consumable_Use_Athena_Att
 bOverrideConcurrency=true
 maxDistance=2250
 numChannels=2
 outputFormat=Stereo
 sampleRate=48000
 soundClassObject=&Fort_Consumables
 volume=0.35
 concurrencyOverrides {bLimitToOwner=true, maxCount=2, resolutionRule=StopOldest, retriggerTime=0.4}
}
