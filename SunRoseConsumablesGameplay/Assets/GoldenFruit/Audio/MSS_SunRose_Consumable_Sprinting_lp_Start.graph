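# One-shot stereo MetaSound for the start of sprinting with the SunRose Golden Fruit consumable.
# On play, a random get picks one of three sprint-start waves for a stereo wave player, whose finish ends the source.
# It uses the consumable attenuation and sound class, slightly lowered pitch and volume, and the concurrency set named for SunRose jumps.
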MetaSoundSource /SunRoseConsumablesGameplay/Assets/GoldenFruit/Audio/MSS_SunRose_Consumable_Sprinting_lp_Start
inputs [
 Sprint Start {type=WaveAsset:Array}
 UE.Source.OnPlay {type=Trigger}
]
interfaces [UE.OutputFormat.Stereo:1.0, UE.Source.OneShot:1.0, UE.Source:1.1]
outputs [
 "UE.OutputFormat.Stereo.Audio:0" {type=Audio}
 "UE.OutputFormat.Stereo.Audio:1" {type=Audio}
 UE.Source.OneShot.OnFinished {type=Trigger}
]
pages [Default {
  input_defaults {
   UE.Source.OnPlay=null
   Sprint Start [
    /SunRoseConsumablesGameplay/Assets/GoldenFruit/Audio/wav/SunRose_Consumable_Sprint_Start_01
    /SunRoseConsumablesGameplay/Assets/GoldenFruit/Audio/wav/SunRose_Consumable_Sprint_Start_02
    /SunRoseConsumablesGameplay/Assets/GoldenFruit/Audio/wav/SunRose_Consumable_Sprint_Start_03
   ]
  }
  nodes [
   ArrayRandomGet_0 = Array.Random Get:WaveAsset:Array(In Array=^Sprint Start, Next=^UE.Source.OnPlay)
   WavePlayer_0 = UE.Wave Player:Stereo(Play=@ArrayRandomGet_0.On Next, Wave Asset=@ArrayRandomGet_0.Value)
  ]
  links [
   {dst="@UE.Source.OneShot.OnFinished", src=WavePlayer_0.On Finished}
   {dst="@UE.OutputFormat.Stereo.Audio:0", src=WavePlayer_0.Out Left}
   {dst="@UE.OutputFormat.Stereo.Audio:1", src=WavePlayer_0.Out Right}
  ]
 }]
properties {
 assetImportData=&self:AssetImportData
 attenuationSettings=&/Att_SunRose_Consumable
 maxDistance=2750
 numChannels=2
 outputFormat=Stereo
 pitch=0.9
 sampleRate=48000
 soundClassObject=&Fort_Consumables
 volume=0.65
 concurrencyOverrides {maxCount=3, resolutionRule=StopOldest, retriggerTime=0.15, volumeScale=0.8}
 concurrencySet [&/Conc_SunRose_Jumps]
}
