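# Additive unlit Niagara sprite material for the Golden Fruit ring glow effect.
# Two Perlin noise samples panning at different scales and directions are multiplied, sharpened, then blended toward particle color by the squared ring texture.
# That blend is tinted by particle color and passed through a shared emissive adjustment function.
# Opacity combines particle alpha with the squared ring mask, with switchable depth fade that the mobile path skips.
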Material /SunRoseConsumablesGameplay/Assets/GoldenFruit/FX/M_GoldenFruitRingGlow
graph {
 properties {bUsedWithNiagaraSprites=true, blendMode=BLEND_Additive, pixelDepthOffsetMode=PDOM_Legacy, shadingModel=MSM_Unlit, usedShadingModels=Unlit}
 nodes [
  n1 = TextureCoordinate()
  n2 = TextureSampleParameter2D(UVs=Panner(Coordinate=Multiply(A=@n1) {constB=0.5}) {speedX=-0.5, speedY=0.4}) {
   parameterName=TexNoise
   samplerType=SAMPLERTYPE_LinearGrayscale
   texture=&T_SimplePerlin_Greyscale
  }
  n3 = TextureSampleParameter2D(UVs=Panner(Coordinate=Multiply(A=@n1) {constB=0.2}) {speedX=0.2, speedY=-0.2}) {
   parameterName=TexNoise
   samplerType=SAMPLERTYPE_LinearGrayscale
   texture=&T_SimplePerlin_Greyscale
  }
  n4 = Power(Base=Multiply(A=@n2.R, B=@n3.R))
  n5 = ParticleColor()
  n6 = TextureSampleParameter2D() {
   parameterName=Tex1
   samplerType=SAMPLERTYPE_LinearGrayscale
   texture=&/T_GoldenFruit_RingGlow
  }
  n7 = Multiply(A=@n6.R, B=@n6.R)
  MaterialFunctionCall(Emissive=Multiply(A=LinearInterpolate(A=@n4, B=@n5.RGB, Alpha=@n7), B=@n5.RGB)) {materialFunction=&MF_FX_EmissiveAdjustments}
  n8 = Multiply(A=@n5.A, B=@n7)
  n9 = StaticSwitchParameter(True=DepthFade(Opacity=@n8, FadeDistance=ScalarParameter() {group=Opacity, parameterName=Depth Fade Distance}), False=@n8) {group=Opacity, parameterName=Enable Depth Fade}
  ShadingPathSwitch(Default=@n9, Deferred=@n9, Mobile=@n8) {desc=Migrated from FeatureLevelSwitch}
 ]
}
