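# Material function that produces a scrolling, grainy gold sand mask for Golden Fruit effects.
# It samples one sand noise texture at two scales, the base tiling and double that, both panned by shared speed and time inputs.
# The two samples are multiplied, contrast-boosted with CheapContrast, saturated and scaled by an intensity input.
# Tiling, pan speed, time and intensity are function inputs, with time previewing as Time.
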MaterialFunction /SunRoseConsumablesGameplay/Assets/GoldenFruit/FX/MF_GoldenFruit_Sand
graph {nodes [
  n1 = Constant() {r=4}
  n2 = Constant() {r=1}
  n3 = Multiply(
   A=TextureCoordinate()
   B=AppendVector(
    A=FunctionInput(Preview=@n1) {inputName=GoldSand_U_Tile, inputType=FunctionInput_Scalar}
    B=FunctionInput(Preview=@n2) {inputName=GoldSand_V_Tile, inputType=FunctionInput_Scalar}
   )
  )
  n4 = FunctionInput(Preview=Time()) {inputName=PannerTimeInput, inputType=FunctionInput_Scalar}
  n5 = AppendVector(
   A=FunctionInput(Preview=Constant() {r=-0.5}) {inputName=PannerSpeedU, inputType=FunctionInput_Scalar}
   B=FunctionInput(Preview=Constant()) {inputName=PannerSpeedV, inputType=FunctionInput_Scalar}
  )
  n6 = Panner(Coordinate=@n3, Time=@n4, Speed=@n5) {speedX=-0.5}
  n7 = Panner(Coordinate=Divide(A=@n3, B=Constant() {r=0.5}), Time=@n4, Speed=@n5) {speedX=-0.5}
  n8 = MaterialFunctionCall(
   In=Multiply(
    A=TextureSample(UVs=@n6) {samplerType=SAMPLERTYPE_LinearGrayscale, texture=/Game/Global/Textures/Default/Noise/T_SandTime2}.R
    B=TextureSample(UVs=@n7) {samplerType=SAMPLERTYPE_LinearGrayscale, texture=&T_SandTime2}.R
   )
   Contrast=Constant() {r=1}
  ) {materialFunction=&CheapContrast}
  n9 = Constant() {r=10}
  FunctionOutput(None=Multiply(A=Saturate(None=@n8.Result), B=FunctionInput(Preview=@n9) {inputName=GoldSandIntensity, inputType=FunctionInput_Scalar}))
 ]}
