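# Translucent unlit Niagara sprite material for the sand smoke effect of the Royal Golden Fruit consumable.
# Particle color feeds the emissive adjustments function, while a squared smoke texture, modulated by an animated gold sand function, forms opacity with particle alpha.
# A dynamic Dissolve parameter erodes a mask built from repeated radial gradients and the smoke, and depth fade is bypassed on mobile.
# SandSettings' ScaleX drives both sand tiling and both panner speeds.
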Material /SunRoseConsumablesGameplay/Assets/GoldenFruit/FX/M_GoldenFruitRoyal_SandSmoke
graph {
 properties {bUsedWithNiagaraSprites=true, blendMode=BLEND_Translucent, pixelDepthOffsetMode=PDOM_Legacy, shadingModel=MSM_Unlit, usedShadingModels=Unlit}
 nodes [
  n1 = ParticleColor()
  MaterialFunctionCall(Emissive=@n1.RGB) {materialFunction=&MF_FX_EmissiveAdjustments}
  n2 = TextureSampleParameter2D() {
   parameterName=SmokeTex
   samplerType=SAMPLERTYPE_LinearGrayscale
   texture=&/T_GoldenFruit_SandCloudBurst
  }
  n3 = Multiply(A=@n2.R, B=@n2.R)
  n4 = VectorParameter() {parameterName=SandSettings, channelNames {a=PanY, b=PanX, g=ScaleY, r=ScaleX}, defaultValue=<rgba 4 4 0 0>}
  n5 = ScalarParameter() {defaultValue=5, parameterName=GoldSandIntensity}
  n6 = Time()
  n7 = Add(
   A=Multiply(
    A=@n3
    B=MaterialFunctionCall!(
     GoldSand_U_Tile=@n4.ScaleX
     GoldSand_V_Tile=@n4.ScaleX
     GoldSandIntensity=@n5
     PannerSpeedU=@n4.ScaleX
     PannerSpeedV=@n4.ScaleX
     PannerTimeInput=@n6
    ) {materialFunction=&/MF_GoldenFruit_Sand}.Result
   )
   B=@n3
  ) {constB=0.3}
  n8 = DepthFade(FadeDistance=ScalarParameter() {defaultValue=5, parameterName=FadeDistance})
  n9 = Constant() {r=1}
  n10 = Multiply(A=TextureCoordinate()) {constB=2}
  n11 = Length(None=Add(A=Frac(None=@n10)) {constB=-0.5}) {is=class MaterialExpressionLength}
  n12 = Multiply(A=@n3)
  n13 = Add(A=Saturate(None=Subtract(A=@n11, B=@n12)))
  n14 = Multiply(A=DynamicParameter() {defaultValue=<rgba 0 1 1 1>, paramNames [Dissolve,Param2,Param3,Param4]}.Dissolve) {constB=2}
  Multiply(
   A=Multiply(A=Multiply(A=@n1.A, B=@n7), B=ShadingPathSwitch(Default=@n8, Deferred=@n8, Mobile=@n9) {desc=Migrated from FeatureLevelSwitch})
   B=Saturate(None=Subtract(A=@n13, B=@n14))
  )
 ]
}
