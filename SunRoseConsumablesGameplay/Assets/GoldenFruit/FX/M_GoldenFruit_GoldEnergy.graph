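# Additive unlit material for Niagara mesh particles that draws animated gold energy for the Golden Fruit consumable effect.
# Panning noise distorts the UVs of a plasma texture and a soft noise texture.
# Plasma, distortion and a dynamic AnimOffset form a banded mask, sparkled by a gold sand function except on mobile, driving particle-tinted emissive.
# Particle alpha, soft noise and a V-coordinate gradient multiply into an opacity term.
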Material /SunRoseConsumablesGameplay/Assets/GoldenFruit/FX/M_GoldenFruit_GoldEnergy
graph {
 properties {
  bUsedWithNiagaraMeshParticles=true
  blendMode=BLEND_Additive
  pixelDepthOffsetMode=PDOM_Legacy
  shadingModel=MSM_Unlit
  usedShadingModels=Unlit
 }
 nodes [
  n1 = ParticleColor()
  n2 = TextureCoordinate()
  n3 = VectorParameter() {parameterName=OpacityNoiseUV, channelNames {a=PanY, b=PanX, g=ScaleY, r=ScaleX}, defaultValue=<rgba 2 0.5 0 0.5>}
  n4 = MaterialFunctionCall(X=@n3.ScaleX, Y=@n3.ScaleX) {materialFunction=&MakeFloat2}
  n5 = MaterialFunctionCall(X=@n3.PanX, Y=@n3.PanX) {materialFunction=&MakeFloat2}
  n6 = Add(A=Time())
  n7 = VectorParameter() {parameterName=DistoNoiseUV, channelNames {a=PanY, b=PanX, g=ScaleY, r=ScaleX}, defaultValue=<rgba 1.5 0.4 0 -0.3>}
  n8 = Time()
  n9 = Add(
   A=Multiply(
    A=TextureCoordinate()
    B=MaterialFunctionCall(X=@n7.ScaleX, Y=@n7.ScaleX) {materialFunction=&MakeFloat2}.Result
   )
   B=Multiply(
    A=MaterialFunctionCall(X=@n7.PanX, Y=@n7.PanX) {materialFunction=&MakeFloat2}.Result
    B=Add(A=@n8)
   )
  )
  n10 = Multiply(
   A=ComponentMask(None=TextureSample(UVs=@n9) {samplerType=SAMPLERTYPE_LinearColor, texture=&T_NoisePack_LinearColor}.RGB) {g=true, r=true}
   B=ScalarParameter() {defaultValue=0.125, parameterName=DistoStrength}
  )
  n11 = Add(A=Add(A=Multiply(A=@n2, B=@n4.Result), B=Multiply(A=@n5.Result, B=@n6)), B=@n10)
  n12 = TextureSample(UVs=@n11) {samplerType=SAMPLERTYPE_LinearColor, texture=/Game/Effects/Fort_Effects/Emotes/UltraLight/Texture/T_Plasma_A_C}
  n13 = MaterialFunctionCall!(
   GoldSand_U_Tile=ScalarParameter() {defaultValue=10, parameterName=GoldSand_U_Tile}
   GoldSand_V_Tile=ScalarParameter() {defaultValue=4, parameterName=GoldSand_V_Tile}
   GoldSandIntensity=ScalarParameter() {defaultValue=2, parameterName=GoldSandIntensity}
   PannerSpeedU=Constant()
   PannerSpeedV=Constant() {r=-1}
   PannerTimeInput=Time()
  ) {materialFunction=&/MF_GoldenFruit_Sand}
  n14 = Multiply(A=@n12.R, B=ScalarParameter() {defaultValue=0.125, parameterName=DistoStrength})
  n15 = Multiply(A=ComponentMask(None=@n10) {r=true}) {constB=2}
  n16 = Saturate(None=Add(A=Add(A=@n14, B=@n15), B=DynamicParameter() {paramNames [AnimOffset,Param2,Param3,Param4]}.AnimOffset))
  n17 = Min(A=SmoothStep(Value=@n16) {constMax=0.5}, B=SmoothStep(Value=@n16) {constMax=0.5, constMin=1})
  n18 = Multiply(A=@n17, B=ShadingPathSwitch(Default=@n13.Result, Deferred=@n13.Result, Mobile=Constant() {r=1}) {desc=Migrated from FeatureLevelSwitch})
  MaterialFunctionCall(Emissive=Multiply(A=Multiply(A=@n1.RGB, B=@n12.RGB), B=Add(A=@n17, B=@n18))) {materialFunction=&MF_FX_EmissiveAdjustments}
  n19 = TextureCoordinate()
  Multiply(
   A=@n1.A
   B=Multiply(
    A=TextureSample(UVs=@n11) {texture=/Game/Effects/Fort_Effects/Emotes/UltraLight/Texture/T_SoftNoise_BC7}.R
    B=ComponentMask(None=@n19) {g=true}
   )
  )
 ]
}
